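# Physics asset for the AsterOrder pickaxe cosmetic mesh, simulating a dangling chain that ends in a tag.
# Capsule bodies on a base bone, eight chain links and the tag connect in series through constraints from base_chain down to dyn_tag.
# Every joint has zero twist limit, with a tighter swing cone at the base than along the rest of the chain.
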PhysicsAsset /BRCosmetics/Weapons/FORT_Melee/Pickaxe_AsterOrder/Meshes/AsterOrder_Axe_Physics
bodies {
 base_chain [{shapeType=Capsule, length=0.505, radius=0.505, shapeName=base_chain_capsule, rotation=<rot 0 -90 0>}]
 dyn_chain_a [{shapeType=Capsule, length=1, radius=0.75, shapeName=dyn_chain_a_capsule, center=<1.1 0 0>, rotation=<rot 90 0 0>}]
 dyn_chain_b [{shapeType=Capsule, length=1, radius=0.75, shapeName=dyn_chain_b_capsule, center=<1.1 0 0>, rotation=<rot 90 0 0>}]
 dyn_chain_c [{shapeType=Capsule, length=1, radius=0.75, shapeName=dyn_chain_c_capsule, center=<1.1 0 0>, rotation=<rot 90 0 0>}]
 dyn_chain_d [{shapeType=Capsule, length=1, radius=0.75, shapeName=dyn_chain_d_capsule, center=<1.1 0 0>, rotation=<rot 90 0 0>}]
 dyn_chain_e [{shapeType=Capsule, length=1, radius=0.75, shapeName=dyn_chain_e_capsule, center=<1.1 0 0>, rotation=<rot 90 0 0>}]
 dyn_chain_f [{shapeType=Capsule, length=1, radius=0.75, shapeName=dyn_chain_f_capsule, center=<1.1 0 0>, rotation=<rot 90 0 0>}]
 dyn_chain_g [{shapeType=Capsule, length=1, radius=0.75, shapeName=dyn_chain_g_capsule, center=<1.1 0 0>, rotation=<rot 90 0 0>}]
 dyn_chain_h [{shapeType=Capsule, length=1, radius=0.75, shapeName=dyn_chain_h_capsule, center=<1.1 0 0>, rotation=<rot 90 0 0>}]
 dyn_tag [{shapeType=Capsule, length=1, radius=0.75, shapeName=dyn_tag_capsule, center=<1.1 0 0>, rotation=<rot 90 0 0>}]
}
constraints [
 {bone1Name=dyn_chain_a, bone2Name=base_chain, swing1LimitDegrees=25, swing2LimitDegrees=25, twistLimitDegrees=0}
 {bone1Name=dyn_chain_b, bone2Name=dyn_chain_a, swing1LimitDegrees=45, swing2LimitDegrees=45, twistLimitDegrees=0}
 {bone1Name=dyn_chain_c, bone2Name=dyn_chain_b, swing1LimitDegrees=45, swing2LimitDegrees=45, twistLimitDegrees=0}
 {bone1Name=dyn_chain_d, bone2Name=dyn_chain_c, swing1LimitDegrees=45, swing2LimitDegrees=45, twistLimitDegrees=0}
 {bone1Name=dyn_chain_e, bone2Name=dyn_chain_d, swing1LimitDegrees=45, swing2LimitDegrees=45, twistLimitDegrees=0}
 {bone1Name=dyn_chain_f, bone2Name=dyn_chain_e, swing1LimitDegrees=45, swing2LimitDegrees=45, twistLimitDegrees=0}
 {bone1Name=dyn_chain_g, bone2Name=dyn_chain_f, swing1LimitDegrees=45, swing2LimitDegrees=45, twistLimitDegrees=0}
 {bone1Name=dyn_chain_h, bone2Name=dyn_chain_g, swing1LimitDegrees=45, swing2LimitDegrees=45, twistLimitDegrees=0}
 {bone1Name=dyn_tag, bone2Name=dyn_chain_h, swing1LimitDegrees=45, swing2LimitDegrees=45, twistLimitDegrees=0}
]
